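# Reusable MetaSound patch that plays a one-shot random wave from an input array as mono or stereo.
# Play goes through a Bool trigger compare on Mono_Stereo to one of two random-get branches, True feeding a mono player and False a stereo one.
# Both share no-repeat, reset, pitch shift and stop inputs, with decibel volume converted to linear gain.
# Outputs are separate mono and L/R audio, plus merged finished and nearly-finished triggers.
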MetaSoundPatch /VehicleAudio/Patches/RandomWavePlayer_OneShot
inputs [
 Mono_Stereo {type=Bool}
 No Repeats Number {type=Int32}
 Pitch Shift {type=Float}
 Play {type=Trigger}
 ResetRandomArray {type=Trigger}
 Stop {type=Trigger}
 Volume {type=Float}
 WaveArray {type=WaveAsset:Array}
]
interfaces [MetaSound:1.0]
outputs [
 On Finished {type=Trigger}
 On Nearly Finished {type=Trigger}
 Output L {type=Audio}
 Output Mono {type=Audio}
 Output R {type=Audio}
]
pages [Default {
  input_defaults {Mono_Stereo=null, No Repeats Number=1, Pitch Shift=0, Play=null, ResetRandomArray=null, Stop=null, Volume=0, WaveArray []}
  nodes [
   DecibelstoLinearGain_2 = UE.Decibels to Linear Gain:Float(Decibels=^Volume)
   n1 = TriggerCompare.Clamp:Bool(A=^Mono_Stereo, Compare=^Play)
   ArrayRandomGet_0 = Array.Random Get:WaveAsset:Array(No Repeats=^No Repeats Number, Reset=^ResetRandomArray, In Array=^WaveArray, Next=@n1.True) {also {defaults {Weights []}, version=~}}
   WavePlayer_0 = UE.Wave Player:Mono(Pitch Shift=^Pitch Shift, Stop=^Stop, Play=@ArrayRandomGet_0.On Next, Wave Asset=@ArrayRandomGet_0.Value)
   Multiply_0 = UE.Multiply:Audio by Float(AdditionalOperands=@DecibelstoLinearGain_2.Linear Gain, PrimaryOperand=@WavePlayer_0.Out Mono)
   DecibelstoLinearGain_1 = UE.Decibels to Linear Gain:Float(Decibels=^Volume)
   ArrayRandomGet_1 = Array.Random Get:WaveAsset:Array(No Repeats=^No Repeats Number, Reset=^ResetRandomArray, In Array=^WaveArray, Next=@n1.False) {also {version=~}}
   WavePlayer_1 = UE.Wave Player:Stereo(Pitch Shift=^Pitch Shift, Stop=^Stop, Play=@ArrayRandomGet_1.On Next, Wave Asset=@ArrayRandomGet_1.Value)
   Multiply_1 = UE.Multiply:Audio by Float(AdditionalOperands=@DecibelstoLinearGain_1.Linear Gain, PrimaryOperand=@WavePlayer_1.Out Left)
   DecibelstoLinearGain_0 = UE.Decibels to Linear Gain:Float(Decibels=^Volume)
   Multiply_2 = UE.Multiply:Audio by Float(AdditionalOperands=@DecibelstoLinearGain_0.Linear Gain, PrimaryOperand=@WavePlayer_1.Out Right)
   TriggerAnyTriggerAny2_0 = "TriggerAny.Trigger Any (2)"(In 0=@WavePlayer_0.On Finished, In 1=@WavePlayer_1.On Finished)
   TriggerAnyTriggerAny2_1 = "TriggerAny.Trigger Any (2)"(In 0=@WavePlayer_0.On Nearly Finished, In 1=@WavePlayer_1.On Nearly Finished)
  ]
  links [
   {dst="@Output Mono", src=Multiply_0.Out}
   {dst="@Output L", src=Multiply_1.Out}
   {dst="@Output R", src=Multiply_2.Out}
   {dst="@On Finished", src=TriggerAnyTriggerAny2_0.Out}
   {dst="@On Nearly Finished", src=TriggerAnyTriggerAny2_1.Out}
  ]
 }]
