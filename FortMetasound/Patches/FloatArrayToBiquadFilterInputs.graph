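# MetaSound patch that unpacks a float array of EQ settings into biquad filter frequency, bandwidth, gain and type outputs.
# The ArrayGet trigger runs a comparison on the array's length, and its True output triggers reads of the first four elements in that order.
# The type element is output as a raw float, an Int32 and a BiquadFilterType enum via the FloatToBiquadFilterType patch.
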MetaSoundPatch /FortMetasound/Patches/FloatArrayToBiquadFilterInputs
inputs [
 ArrayGet {type=Trigger}
 EQInputs {type=Float:Array}
]
outputs [
 Bandwidth {type=Float}
 BiquadFilterType {type=Enum:BiquadFilterType}
 BiquadFilterTypeFloat {type=Float}
 BiquadFilterTypeInt32 {type=Int32}
 Freq {type=Float}
 Gain {type=Float}
]
pages [Default {
  input_defaults {ArrayGet=null, EQInputs []}
  nodes [
   n1 = TriggerCompare.Clamp:Int32(Compare=^ArrayGet, A=Array.Num:Float:Array(Array=^EQInputs).Num) {also {defaults {Type=1}}}
   ArrayGet_1 = Array.Get:Float:Array(Array=^EQInputs, Trigger=@n1.True) {also {defaults {Index=1}}}
   ArrayGet_2 = Array.Get:Float:Array(Array=^EQInputs, Trigger=@n1.True)
   ArrayGet_3 = Array.Get:Float:Array(Array=^EQInputs, Trigger=@n1.True) {also {defaults {Index=2}}}
   ArrayGet_0 = Array.Get:Float:Array(Array=^EQInputs, Trigger=@n1.True) {also {defaults {Index=3}}}
   ConvertInt32_0 = Convert.Int32:Float(Float=@ArrayGet_0.Element)
   FloatToBiquadFilterType_0 = &/FloatToBiquadFilterType(Input=@ArrayGet_0.Element)
  ]
  links [
   {dst="@BiquadFilterTypeFloat", src=ArrayGet_0.Element}
   {dst="@Bandwidth", src=ArrayGet_1.Element}
   {dst="@Freq", src=ArrayGet_2.Element}
   {dst="@Gain", src=ArrayGet_3.Element}
   {dst="@BiquadFilterTypeInt32", src=ConvertInt32_0.Int32}
   {dst="@BiquadFilterType", src=FloatToBiquadFilterType_0.Enum:BiquadFilterType}
  ]
 }]
