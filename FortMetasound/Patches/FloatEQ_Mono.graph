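# FloatEQ_Mono is a reusable MetaSound patch that drives a biquad EQ on mono audio from a single float parameter.
# The float is smoothed by InterpTo, then mapped from its input range into separate frequency and gain ranges.
# These set the filter's cutoff and gain, while filter type is exposed and Q is converted to bandwidth.
# It outputs the filtered audio and the computed cutoff frequency.
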MetaSoundPatch /FortMetasound/Patches/FloatEQ_Mono
inputs [
 AudioIn {type=Audio}
 FilterQ {type=Float}
 FilterType {type=Enum:BiquadFilterType}
 FloatInput {type=Float}
 InterpTime {type=Time}
 MaxFloatInput {type=Float}
 MaxFreq {type=Float}
 MaxGain {type=Float}
 MinFloatInput {type=Float}
 MinFreq {type=Float}
 MinGain {type=Float}
]
interfaces [MetaSound:1.0]
outputs [
 AudioOut {type=Audio}
 CutoffFrequency {type=Float}
]
pages [Default {
  input_defaults {
   AudioIn=null
   FilterQ=1
   FilterType=0
   FloatInput=0
   InterpTime=0
   MaxFloatInput=1
   MaxFreq=1
   MaxGain=1
   MinFloatInput=0
   MinFreq=0
   MinGain=0
  }
  nodes [
   ConvertFilterQToBandwidth_0 = UE.Convert Filter Q To Bandwidth(Q=^FilterQ)
   n1 = UE.InterpTo:Audio(Target=^FloatInput, Interp Time=^InterpTime)
   MapRangeMapRange_1 = MapRange.MapRange:Float(In Range B=^MaxFloatInput, Out Range B=^MaxFreq, In Range A=^MinFloatInput, Out Range A=^MinFreq, In=@n1.Value)
   BiquadFilter_0 = UE.Biquad Filter:Audio(
    In=^AudioIn
    Type=^FilterType
    Bandwidth=@ConvertFilterQToBandwidth_0.Bandwidth
    Gain=MapRange.MapRange:Float(In Range B=^MaxFloatInput, Out Range B=^MaxGain, In Range A=^MinFloatInput, Out Range A=^MinGain, In=@n1.Value).Out Value
    Cutoff Frequency=@MapRangeMapRange_1.Out Value
   )
  ]
  links [{dst="@AudioOut", src=BiquadFilter_0.Out}, {dst="@CutoffFrequency", src=MapRangeMapRange_1.Out Value}]
 }]
