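# MSP_WavetableCurve is a MetaSound patch that maps a float input through a curve stored as a wavetable.
# It pulls a table from a WaveTableBank asset by float index, then evaluates that table at the Input value using the chosen interpolation mode.
# The result is a single float output, so parent MetaSounds can use banked wavetables as reusable lookup curves.
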MetaSoundPatch /FortMetasound/Patches/MSP_WavetableCurve
inputs [
 Input {type=Float}
 Interpolation {type=Enum:WaveTableInterpolation}
 WaveTableBank {type=WaveTableBankAsset}
 WaveTableIndex {type=Float}
]
interfaces [MetaSound:1.0]
outputs [Output {type=Float}]
pages [Default {
  input_defaults {Input=0, Interpolation=1, WaveTableBank=null, WaveTableIndex=0}
  nodes [WaveTableEvaluate_0 = UE.WaveTableEvaluate(
    Input=^Input
    Interpolation=^Interpolation
    WaveTable=UE.MetasoundWaveTableGet(WaveTableBank=^WaveTableBank, TableIndex=^WaveTableIndex) {also {version [1,1]}}.Out
   )]
  links [{dst="@Output", src=WaveTableEvaluate_0.Output}]
 }]
