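# MetaSound patch that widens a mono audio input into stereo using a band-split Haas-style delay.
# A two-band splitter at the Crossover frequency sends the low band equally to both channels, keeping it centered.
# The high band goes straight to the left and through a Delay node to the right, with MixLevel driving wet level and its InvertFloat result driving dry.
# A stereo mixer sums both bands into Out L and Out R.
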MetaSoundPatch /FortMetasound/Patches/Multiband_StereoWidener
inputs [
 Crossover {type=Float}
 Delay Time {type=Time}
 Input {type=Audio}
 MixLevel {type=Float}
]
outputs [
 Out L {type=Audio}
 Out R {type=Audio}
]
pages [Default {
  input_defaults {Crossover=494, Delay Time=0.072, Input=null, MixLevel=1}
  nodes [
   BandSplitterBandSplitterMono2_0 = "BandSplitter.Band Splitter (Mono, 2)"(Crossover 0=^Crossover, In=^Input)
   AudioMixerAudioMixerStereo3_0 = "AudioMixer.Audio Mixer (Stereo, 3)"(
    In 0 L=@BandSplitterBandSplitterMono2_0.Band 0 Out
    In 0 R=@BandSplitterBandSplitterMono2_0.Band 0 Out
    In 1 L=@BandSplitterBandSplitterMono2_0.Band 1 Out
    In 1 R=UE.Delay:Audio(
     Delay Time=^Delay Time
     Wet Level=^MixLevel
     In=@BandSplitterBandSplitterMono2_0.Band 1 Out
     Dry Level=&/InvertFloat(FloatIn=^MixLevel).FloatOut
    ).Out
   )
  ]
  links [{dst="@Out L", src=AudioMixerAudioMixerStereo3_0.Out L}, {dst="@Out R", src=AudioMixerAudioMixerStereo3_0.Out R}]
 }]
