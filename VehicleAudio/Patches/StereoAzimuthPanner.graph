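# MetaSound patch for vehicle audio that positions a two-channel input in the stereo field and scales its volume by azimuth.
# AudioL and AudioR each feed an Azimuth Panner with shared settings, with TargetAzimuth negated for the right channel to mirror it.
# A two-input stereo mixer sums both panners into Out L and Out R.
# The Pan and Volume outputs report only the right-channel panner's values.
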MetaSoundPatch /VehicleAudio/Patches/StereoAzimuthPanner
inputs [
 AudioL {type=Audio}
 AudioR {type=Audio}
 Azimuth {type=Float}
 AzimuthPanOffset {type=Float}
 AzimuthVolumeRange {type=Float}
 InvertPan {type=Bool}
 MinVolume {type=Float}
 PanRange {type=Float}
 TargetAzimuth {type=Float}
]
outputs [
 Out L {type=Audio}
 Out R {type=Audio}
 Pan {type=Float}
 Volume {type=Float}
]
pages [Default {
  input_defaults {
   AudioL=null
   AudioR=null
   Azimuth=0
   AzimuthPanOffset=90
   AzimuthVolumeRange=180
   InvertPan="<MetaSoundBuilderResult.SUCCEEDED: 0>"
   MinVolume=0
   PanRange=1
   TargetAzimuth=0
  }
  nodes [
   FortniteAzimuthPanner_0 = Fortnite.AzimuthPanner:Audio(
    Audio=^AudioR
    Azimuth=^Azimuth
    AzimuthPanOffset=^AzimuthPanOffset
    AzimuthVolumeRange=^AzimuthVolumeRange
    InvertPan=^InvertPan
    MinVolume=^MinVolume
    PanRange=^PanRange
    TargetAzimuth=UE.Multiply:Float(PrimaryOperand=^TargetAzimuth) {also {defaults {AdditionalOperands=-1}}}.Out
   )
   n1 = Fortnite.AzimuthPanner:Audio(
    Audio=^AudioL
    Azimuth=^Azimuth
    AzimuthPanOffset=^AzimuthPanOffset
    AzimuthVolumeRange=^AzimuthVolumeRange
    InvertPan=^InvertPan
    MinVolume=^MinVolume
    PanRange=^PanRange
    TargetAzimuth=^TargetAzimuth
   )
   AudioMixerAudioMixerStereo2_0 = "AudioMixer.Audio Mixer (Stereo, 2)"(In 1 L=@FortniteAzimuthPanner_0.Left, In 1 R=@FortniteAzimuthPanner_0.Right, In 0 L=@n1.Left, In 0 R=@n1.Right)
  ]
  links [
   {dst="@Out L", src=AudioMixerAudioMixerStereo2_0.Out L}
   {dst="@Out R", src=AudioMixerAudioMixerStereo2_0.Out R}
   {dst="@Pan", src=FortniteAzimuthPanner_0.Pan}
   {dst="@Volume", src=FortniteAzimuthPanner_0.Volume}
  ]
 }]
