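# MetaSound patch that mixes stereo engine, exhaust and interior layers for vehicle audio, shaping engine and exhaust by source azimuth.
# Engine and exhaust each pass through a StereoAzimuthPanner, whose pan range varies with absolute azimuth and whose volume has a decibel-set floor.
# Each panner's volume also drives an interpolated biquad cutoff, so tone follows orientation.
# Interior audio bypasses this processing into the final mixer, where inverted mute booleans gate all three layers.
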MetaSoundPatch /VehicleAudio/Patches/DM_EngineAudio_AzimuthPanner
inputs [
 Engine_MaxFreq {type=Float, access=Constructor}
 Engine_MinFreq {type=Float, access=Constructor}
 Engine_MinVolume {type=Float, access=Constructor}
 EnginefilterType {type=Enum:BiquadFilterType, access=Constructor}
 EngineL {type=Audio}
 EngineMute {type=Bool}
 EngineR {type=Audio}
 Exhaust_MaxFreq {type=Float, access=Constructor}
 Exhaust_MinFreq {type=Float, access=Constructor}
 Exhaust_MinVolume {type=Float, access=Constructor}
 ExhaustFilterType {type=Enum:BiquadFilterType}
 ExhaustL {type=Audio}
 ExhaustMute {type=Bool}
 ExhaustR {type=Audio}
 InteriorL {type=Audio}
 InteriorMute {type=Bool}
 InteriorR {type=Audio}
 UE.Source.Orientation.Azimuth {type=Float}
 UE.Source.Orientation.Elevation {type=Float}
]
interfaces [UE.Source.Orientation:1.0]
outputs [
 Out L {type=Audio}
 Out R {type=Audio}
]
pages [Default {
  input_defaults {
   EngineL=null
   EngineMute=null
   EngineR=null
   Engine_MaxFreq=20000
   Engine_MinFreq=1
   Engine_MinVolume=0
   EnginefilterType=0
   ExhaustFilterType=1
   ExhaustL=null
   ExhaustMute=null
   ExhaustR=null
   Exhaust_MaxFreq=20
   Exhaust_MinFreq=500
   Exhaust_MinVolume=0
   InteriorL=null
   InteriorMute=null
   InteriorR=null
   UE.Source.Orientation.Azimuth=0
   UE.Source.Orientation.Elevation=0
  }
  nodes [
   DecibelstoLinearGain_1 = UE.Decibels to Linear Gain:Float(Decibels=^Engine_MinVolume)
   n1 = &/StereoAzimuthPanner(
    AudioL=^EngineL
    AudioR=^EngineR
    Azimuth=^UE.Source.Orientation.Azimuth
    MinVolume=@DecibelstoLinearGain_1.Linear Gain
    PanRange=MapRange.MapRange:Float(In=AbsoluteValue.Abs:Float(Input=^UE.Source.Orientation.Azimuth).Value) {also {defaults {In Range A=110, In Range B=180, Out Range A=0.3, Out Range B=0}}}.Out Value
    TargetAzimuth=VariableAccessor.Float(Variable=InitVariable.Float().Variable).Value
   ) {also {defaults {AzimuthVolumeRange=90}}}
   n2 = &StereoBiquadFilter(
    Type=^EnginefilterType
    Cutoff Frequency=UE.InterpTo:Audio(Target=MapRange.MapRange:Float(Out Range B=^Engine_MaxFreq, Out Range A=^Engine_MinFreq, In Range A=@DecibelstoLinearGain_1.Linear Gain, In=@n1.Volume).Out Value).Value
    In L=@n1.Out L
    In R=@n1.Out R
   )
   DecibelstoLinearGain_0 = UE.Decibels to Linear Gain:Float(Decibels=^Exhaust_MinVolume)
   n3 = &StereoAzimuthPanner(
    AudioL=^ExhaustL
    AudioR=^ExhaustR
    Azimuth=^UE.Source.Orientation.Azimuth
    MinVolume=@DecibelstoLinearGain_0.Linear Gain
    PanRange=MapRange.MapRange:Float(In=AbsoluteValue.Abs:Float(Input=^UE.Source.Orientation.Azimuth).Value) {also {defaults {In Range B=60, Out Range B=0.3}}}.Out Value
    TargetAzimuth=VariableAccessor.Float(Variable=InitVariable.Float().Variable).Value
   ) {also {defaults {AzimuthVolumeRange=160}}}
   n4 = &StereoBiquadFilter(
    Type=^ExhaustFilterType
    Cutoff Frequency=UE.InterpTo:Audio(Target=MapRange.MapRange:Float(Out Range B=^Exhaust_MaxFreq, Out Range A=^Exhaust_MinFreq, In Range A=@DecibelstoLinearGain_0.Linear Gain, In=@n3.Volume).Out Value).Value
    In L=@n3.Out L
    In R=@n3.Out R
   )
   AudioMixerAudioMixerStereo3_0 = "AudioMixer.Audio Mixer (Stereo, 3)"(
    In 2 L=^InteriorL
    In 2 R=^InteriorR
    Gain 1=/FortMetasound/Patches/InvertFloat(FloatIn=Convert.Float:Bool(Bool=^ExhaustMute).Float).FloatOut
    Gain 0=&InvertFloat(FloatIn=Convert.Float:Bool(Bool=^EngineMute).Float).FloatOut
    Gain 2=&InvertFloat(FloatIn=Convert.Float:Bool(Bool=^InteriorMute).Float).FloatOut
    In 0 L=@n2.Out L
    In 0 R=@n2.Out R
    In 1 L=@n4.Out L
    In 1 R=@n4.Out R
   )
  ]
  links [{dst="@Out L", src=AudioMixerAudioMixerStereo3_0.Out L}, {dst="@Out R", src=AudioMixerAudioMixerStereo3_0.Out R}]
 }]
