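# Mono one-shot MetaSound source that plays a vehicle engine blowoff sound from a shuffled wave array, with adjustable pitch.
# EngineBlowoff triggers playback after an array-size check and either an IsLocal or character affiliation check.
# An ADSR envelope, attacked by the blowoff and released by a throttle-change comparison, scales the decibel-based output volume.
# OnFinished fires when the wave ends or the IsLocal check fails.
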MetaSoundSource /VehicleAudio/Patches/DM_EngineAudio_Blowoff
inputs [
 BlowoffPitch {type=Float}
 BlowoffSound {type=WaveAsset:Array, access=Constructor}
 Character.Affiliation {type=Enum:FortAudioAffiliation}
 Character.Health {type=Float}
 Character.UniqueID {type=Int32}
 EngineBlowoff {type=Trigger}
 IsLocal {type=Bool}
 OutputVolume {type=Float, access=Constructor}
 REV.EnableShifting {type=Bool}
 REV.Gear {type=Int32}
 REV.RPM {type=Float}
 REV.Throttle {type=Float}
 REV.Velocity {type=Float}
 UE.Source.OnPlay {type=Trigger}
]
interfaces [Fort.Character:1.0, REV:1.0, UE.OutputFormat.Mono:1.0, UE.Source.OneShot:1.0, UE.Source:1.1]
outputs [
 "UE.OutputFormat.Mono.Audio:0" {type=Audio}
 UE.Source.OneShot.OnFinished {type=Trigger}
]
pages [Default {
  input_defaults {
   BlowoffPitch=0
   Character.Affiliation=0
   Character.Health=0
   Character.UniqueID=0
   EngineBlowoff=null
   IsLocal=null
   OutputVolume=0
   REV.EnableShifting=null
   REV.Gear=1
   REV.RPM=0
   REV.Throttle=0
   REV.Velocity=0
   UE.Source.OnPlay=null
   BlowoffSound []
  }
  nodes [
   InitVariable.Float()
   TriggerOnValueChange_0 = UE.Trigger On Value Change:Float(Value=^REV.Throttle)
   ADSREnvelopeADSREnvelope_0 = ADSR Envelope.ADSR Envelope:Float(
    Trigger Attack=^EngineBlowoff
    Trigger Release=TriggerCompare.Clamp:Float(A=^REV.Throttle, Compare=@TriggerOnValueChange_0.Trigger) {also {defaults {B=1}}}.True
   ) {also {defaults {Release Time=0.35, Sustain Level=1}}}
   DecibelstoLinearGain_0 = UE.Decibels to Linear Gain:Float(Decibels=^OutputVolume)
   n1 = TriggerCompare.Clamp:Int32(Compare=^EngineBlowoff, A=Array.Num:WaveAsset:Array(Array=^BlowoffSound).Num) {also {defaults {Type=1}}}
   n2 = TriggerCompare.Clamp:Bool(A=^IsLocal, Compare=@n1.True) {also {defaults {B="<MetaSoundBuilderResult.SUCCEEDED: 0>"}}}
   TriggerAnyTriggerAny2_0 = "TriggerAny.Trigger Any (2)"(In 1=&TriggerCompareAffiliation(A=^Character.Affiliation, Compare=@n1.True).True, In 0=@n2.True)
   n3 = Array.Shuffle:WaveAsset:Array(In Array=^BlowoffSound, Next=@TriggerAnyTriggerAny2_0.Out)
   WavePlayer_0 = UE.Wave Player:Mono(Pitch Shift=^BlowoffPitch, Play=@n3.On Next, Wave Asset=@n3.Value) {also {version [1,1]}}
   Multiply_0 = UE.Multiply:Audio by Float(
    AdditionalOperands=UE.Multiply:Float(AdditionalOperands=@ADSREnvelopeADSREnvelope_0.Out Envelope, PrimaryOperand=@DecibelstoLinearGain_0.Linear Gain).Out
    PrimaryOperand=@WavePlayer_0.Out Mono
   )
   TriggerAnyTriggerAny2_1 = "TriggerAny.Trigger Any (2)"(In 1=UE.Reroute(Value=UE.Reroute(Value=@n2.False).Value).Value, In 0=@WavePlayer_0.On Finished)
  ]
  links [{dst="@UE.OutputFormat.Mono.Audio:0", src=Multiply_0.Out}, {dst="@UE.Source.OneShot.OnFinished", src=TriggerAnyTriggerAny2_1.Out}]
 }]
properties {assetImportData=&self:AssetImportData, sampleRate=48000}
