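# One-shot mono MetaSound source that plays vehicle exhaust pops on upshift, downshift and blowoff triggers.
# Each trigger passes an IsLocal or character-affiliation check and a non-empty array check, with blowoff also checking ExhaustPop.OnBlowoff.
# After a GearChangeTime delay, shuffled short pops play for upshift and blowoff, long ones for downshift.
# A mixer applies separate shift and blowoff dB gains, and any player finishing fires OnFinished.
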MetaSoundSource /VehicleAudio/Patches/DM_EngineAudio_ExhaustPops
inputs [
 Blowoff {type=Trigger}
 BlowoffVolume {type=Float}
 Character.Affiliation {type=Enum:FortAudioAffiliation}
 Character.Health {type=Float}
 Character.UniqueID {type=Int32}
 Downshift {type=Trigger}
 ExhaustPop.OnBlowoff {type=Bool}
 ExhaustPops_Long {type=WaveAsset:Array, access=Constructor}
 ExhaustPops_Short {type=WaveAsset:Array, access=Constructor}
 GearChangeTime {type=Time, access=Constructor}
 IsLocal {type=Bool}
 Pitch {type=Float, access=Constructor}
 ShiftVolume {type=Float, access=Constructor}
 UE.Source.OnPlay {type=Trigger}
 Upshift {type=Trigger}
]
interfaces [Fort.Character:1.0, UE.OutputFormat.Mono:1.0, UE.Source.OneShot:1.0, UE.Source:1.1]
outputs [
 "UE.OutputFormat.Mono.Audio:0" {type=Audio}
 UE.Source.OneShot.OnFinished {type=Trigger}
]
pages [Default {
  input_defaults {
   Blowoff=null
   BlowoffVolume=0
   Character.Affiliation=0
   Character.Health=0
   Character.UniqueID=0
   Downshift=null
   ExhaustPop.OnBlowoff=null
   GearChangeTime=0.2
   IsLocal=null
   Pitch=0
   ShiftVolume=0
   UE.Source.OnPlay=null
   Upshift=null
   ExhaustPops_Long []
   ExhaustPops_Short []
  }
  nodes [
   DecibelstoLinearGain_0 = UE.Decibels to Linear Gain:Float(Decibels=^BlowoffVolume)
   DecibelstoLinearGain_1 = UE.Decibels to Linear Gain:Float(Decibels=^ShiftVolume)
   n1 = TriggerCompare.Clamp:Bool(Compare=^Blowoff, A=^ExhaustPop.OnBlowoff) {also {defaults {B="<MetaSoundBuilderResult.SUCCEEDED: 0>"}}}
   TriggerAnyTriggerAny2_0 = "TriggerAny.Trigger Any (2)"(
    In 1=&TriggerCompareAffiliation(A=^Character.Affiliation, Compare=@n1.True).True
    In 0=TriggerCompare.Clamp:Bool(A=^IsLocal, Compare=@n1.True) {also {defaults {B="<MetaSoundBuilderResult.SUCCEEDED: 0>"}}}.True
   )
   TriggerDelay_1 = UE.Trigger Delay(Delay Time=^GearChangeTime, In=@TriggerAnyTriggerAny2_0.Out)
   n2 = Array.Shuffle:WaveAsset:Array(
    In Array=^ExhaustPops_Short
    Next=TriggerCompare.Clamp:Int32(A=Array.Num:WaveAsset:Array(Array=^ExhaustPops_Short).Num, Compare=@TriggerDelay_1.Out) {also {defaults {B=0, Type=1}}}.True
   )
   WavePlayer_0 = UE.Wave Player:Mono(Pitch Shift=^Pitch, Play=@n2.On Next, Wave Asset=@n2.Value) {also {version [1,1]}}
   n3 = TriggerCompare.Clamp:Int32(Compare=^Downshift, A=Array.Num:WaveAsset:Array(Array=^ExhaustPops_Long).Num) {also {defaults {B=0, Type=1}}}
   TriggerAnyTriggerAny2_2 = "TriggerAny.Trigger Any (2)"(
    In 1=&TriggerCompareAffiliation(A=^Character.Affiliation, Compare=@n3.True).True
    In 0=TriggerCompare.Clamp:Bool(A=^IsLocal, Compare=@n3.True) {also {defaults {B="<MetaSoundBuilderResult.SUCCEEDED: 0>"}}}.True
   )
   TriggerDelay_2 = UE.Trigger Delay(Delay Time=^GearChangeTime, In=@TriggerAnyTriggerAny2_2.Out)
   n4 = Array.Shuffle:WaveAsset:Array(In Array=^ExhaustPops_Long, Next=@TriggerDelay_2.Out)
   WavePlayer_1 = UE.Wave Player:Mono(Pitch Shift=^Pitch, Play=@n4.On Next, Wave Asset=@n4.Value) {also {version [1,1]}}
   n5 = TriggerCompare.Clamp:Int32(Compare=^Upshift, A=Array.Num:WaveAsset:Array(Array=^ExhaustPops_Short).Num) {also {defaults {B=0, Type=1}}}
   TriggerAnyTriggerAny2_1 = "TriggerAny.Trigger Any (2)"(
    In 1=&TriggerCompareAffiliation(A=^Character.Affiliation, Compare=@n5.True).True
    In 0=TriggerCompare.Clamp:Bool(A=^IsLocal, Compare=@n5.True) {also {defaults {B="<MetaSoundBuilderResult.SUCCEEDED: 0>"}}}.True
   )
   TriggerDelay_0 = UE.Trigger Delay(Delay Time=^GearChangeTime, In=@TriggerAnyTriggerAny2_1.Out)
   n6 = Array.Shuffle:WaveAsset:Array(In Array=^ExhaustPops_Short, Next=@TriggerDelay_0.Out)
   WavePlayer_2 = UE.Wave Player:Mono(Pitch Shift=^Pitch, Play=@n6.On Next, Wave Asset=@n6.Value) {also {version [1,1]}}
   AudioMixerAudioMixerMono3_0 = "AudioMixer.Audio Mixer (Mono, 3)"(
    Gain 2=@DecibelstoLinearGain_0.Linear Gain
    Gain 0=@DecibelstoLinearGain_1.Linear Gain
    Gain 1=@DecibelstoLinearGain_1.Linear Gain
    In 2=@WavePlayer_0.Out Mono
    In 1=@WavePlayer_1.Out Mono
    In 0=@WavePlayer_2.Out Mono
   )
   TriggerAnyTriggerAny3_0 = "TriggerAny.Trigger Any (3)"(
    In 2=UE.Reroute(Value=@WavePlayer_0.On Finished).Value
    In 0=UE.Reroute(Value=@WavePlayer_2.On Finished).Value
    In 1=UE.Reroute(Value=@WavePlayer_1.On Finished).Value
   )
  ]
  links [
   {dst="@UE.OutputFormat.Mono.Audio:0", src=AudioMixerAudioMixerMono3_0.Out}
   {dst="@UE.Source.OneShot.OnFinished", src=TriggerAnyTriggerAny3_0.Out}
  ]
 }]
properties {assetImportData=&self:AssetImportData, sampleRate=48000}
