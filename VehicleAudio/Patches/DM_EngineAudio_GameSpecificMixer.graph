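# MetaSound Source that mixes vehicle engine audio stems into stereo Engine, Exhaust and Interior outputs.
# Engine, Intake and Transmission feed the Engine bus, Exhaust feeds its own bus, and Interior channels pass straight through.
# The SynthPerspective bool, converted to float and inverted for the exhaust mixer, routes stereo synth to Engine when true or Exhaust when false.
# The interface's mono audio output stays unconnected.
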MetaSoundSource /VehicleAudio/Patches/DM_EngineAudio_GameSpecificMixer
inputs [
 Engine {type=Audio}
 Exhaust {type=Audio}
 Intake {type=Audio}
 InteriorL {type=Audio}
 InteriorR {type=Audio}
 SynthL {type=Audio}
 SynthPerspective {type=Bool}
 SynthR {type=Audio}
 Transmission {type=Audio}
 UE.Source.OnPlay {type=Trigger}
]
interfaces [UE.OutputFormat.Mono:1.0, UE.Source:1.1]
outputs [
 EngineL {type=Audio}
 EngineR {type=Audio}
 ExhaustL {type=Audio}
 ExhaustR {type=Audio}
 InteriorL {type=Audio}
 InteriorR {type=Audio}
 "UE.OutputFormat.Mono.Audio:0" {type=Audio}
]
pages [Default {
  input_defaults {
   Engine=null
   Exhaust=null
   Intake=null
   InteriorL=null
   InteriorR=null
   SynthL=null
   SynthPerspective=null
   SynthR=null
   Transmission=null
   UE.Source.OnPlay=null
  }
  nodes [
   AudioMixerAudioMixerStereo3_0 = "AudioMixer.Audio Mixer (Stereo, 3)"(
    In 0 L=^Exhaust
    In 0 R=^Exhaust
    In 1 L=^SynthL
    In 1 R=^SynthR
    Gain 1=/FortMetasound/Patches/InvertFloat(FloatIn=Convert.Float:Bool(Bool=^SynthPerspective).Float).FloatOut
   ) {also {defaults {Gain 0=0.7}}}
   AudioMixerAudioMixerStereo8_0 = "AudioMixer.Audio Mixer (Stereo, 8)"(
    In 0 L=^Engine
    In 0 R=^Engine
    In 6 L=^Intake
    In 6 R=^Intake
    In 1 L=^SynthL
    In 1 R=^SynthR
    In 7 L=^Transmission
    In 7 R=^Transmission
    Gain 1=Convert.Float:Bool(Bool=^SynthPerspective).Float
   ) {also {defaults {Gain 0=0.7, Gain 6=0.7, Gain 7=0.7}}}
  ]
  links [
   {dst="@InteriorL", src="@InteriorL"}
   {dst="@InteriorR", src="@InteriorR"}
   {dst="@ExhaustL", src=AudioMixerAudioMixerStereo3_0.Out L}
   {dst="@ExhaustR", src=AudioMixerAudioMixerStereo3_0.Out R}
   {dst="@EngineL", src=AudioMixerAudioMixerStereo8_0.Out L}
   {dst="@EngineR", src=AudioMixerAudioMixerStereo8_0.Out R}
  ]
 }]
properties {assetImportData=&self:AssetImportData, bLooping=true, duration=10000, sampleRate=48000}
