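# MetaSound source for vehicle engine audio that applies mix-perspective EQ to stereo engine, exhaust and interior signals.
# On play, helper patches unpack exterior and interior EQ arrays into biquad frequency, bandwidth, gain and type for engine and exhaust.
# An interpolated MixP value crossfades these parameters from exterior (0) to interior (1), driving one stereo filter per signal pair.
# The interior pair's filter takes no EQ parameters, and the mono output is unconnected.
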MetaSoundSource /VehicleAudio/Patches/DM_EngineAudio_MixPFilters
inputs [
 EngineL {type=Audio}
 EngineR {type=Audio}
 ExhaustL {type=Audio}
 ExhaustR {type=Audio}
 InteriorL {type=Audio}
 InteriorR {type=Audio}
 MixP {type=Int32}
 MixP.Exterior.Engine.EQ {type=Float:Array, access=Constructor}
 MixP.Exterior.Exhaust.EQ {type=Float:Array, access=Constructor}
 MixP.Interior.Engine.EQ {type=Float:Array, access=Constructor}
 MixP.Interior.Exhaust.EQ {type=Float:Array, access=Constructor}
 UE.Source.OnPlay {type=Trigger}
]
interfaces [UE.OutputFormat.Mono:1.0, UE.Source:1.1]
outputs [
 EngineL {type=Audio}
 EngineR {type=Audio}
 ExhaustL {type=Audio}
 ExhaustR {type=Audio}
 InteriorL {type=Audio}
 InteriorR {type=Audio}
 "UE.OutputFormat.Mono.Audio:0" {type=Audio}
]
pages [Default {
  input_defaults {
   EngineL=null
   EngineR=null
   ExhaustL=null
   ExhaustR=null
   InteriorL=null
   InteriorR=null
   MixP=1
   UE.Source.OnPlay=null
   MixP.Exterior.Engine.EQ [20,1,0,1]
   MixP.Exterior.Exhaust.EQ [20,1,0,1]
   MixP.Interior.Engine.EQ [20,1,0,1]
   MixP.Interior.Exhaust.EQ [20,1,0,1]
  }
  nodes [
   n1 = /FortMetasound/Patches/FloatArrayToBiquadFilterInputs(EQInputs=^MixP.Interior.Engine.EQ, ArrayGet=^UE.Source.OnPlay)
   n2 = &FloatArrayToBiquadFilterInputs(EQInputs=^MixP.Exterior.Engine.EQ, ArrayGet=^UE.Source.OnPlay)
   n3 = VariableAccessor.Time(Variable=InitVariable.Time().Variable)
   n4 = UE.InterpTo:Audio(Target=Convert.Float:Int32(Int32=^MixP).Float, Interp Time=VariableAccessor.Time(Variable=@n3.Variable).Value)
   CrossfadeTriggerRouteFloat8_2 = "Crossfade.Trigger Route (Float, 8):Float"(In 1=@n1.Bandwidth, In 0=@n2.Bandwidth, Crossfade Value=@n4.Value) {also {defaults {In 2=0, In 3=0, In 4=0, In 5=0, In 6=0, In 7=0}}}
   CrossfadeTriggerRouteFloat8_5 = "Crossfade.Trigger Route (Float, 8):Float"(In 1=@n1.Freq, In 0=@n2.Freq, Crossfade Value=@n4.Value) {also {defaults {In 2=0, In 3=0, In 4=0, In 5=0, In 6=0, In 7=0}}}
   CrossfadeTriggerRouteFloat8_6 = "Crossfade.Trigger Route (Float, 8):Float"(In 1=@n1.Gain, In 0=@n2.Gain, Crossfade Value=@n4.Value) {also {defaults {In 2=0, In 3=0, In 4=0, In 5=0, In 6=0, In 7=0}}}
   CrossfadeTriggerRouteFloat8_0 = "Crossfade.Trigger Route (Float, 8):Float"(In 1=@n1.BiquadFilterTypeFloat, In 0=@n2.BiquadFilterTypeFloat, Crossfade Value=@n4.Value) {also {defaults {In 2=0, In 3=0, In 4=0, In 5=0, In 6=0, In 7=0}}}
   StereoBiquadFilter_0 = &StereoBiquadFilter(
    In L=^EngineL
    In R=^EngineR
    Bandwidth=@CrossfadeTriggerRouteFloat8_2.Out
    Cutoff Frequency=@CrossfadeTriggerRouteFloat8_5.Out
    Gain=@CrossfadeTriggerRouteFloat8_6.Out
    Type=/FortMetasound/Patches/FloatToBiquadFilterType(Input=@CrossfadeTriggerRouteFloat8_0.Out)."Enum:BiquadFilterType"
   )
   n5 = &FloatArrayToBiquadFilterInputs(EQInputs=^MixP.Interior.Exhaust.EQ, ArrayGet=^UE.Source.OnPlay)
   n6 = &FloatArrayToBiquadFilterInputs(EQInputs=^MixP.Exterior.Exhaust.EQ, ArrayGet=^UE.Source.OnPlay)
   n7 = UE.InterpTo:Audio(Target=Convert.Float:Int32(Int32=^MixP).Float, Interp Time=@n3.Value)
   CrossfadeTriggerRouteFloat8_1 = "Crossfade.Trigger Route (Float, 8):Float"(In 1=@n5.Freq, In 0=@n6.Freq, Crossfade Value=@n7.Value)
   CrossfadeTriggerRouteFloat8_4 = "Crossfade.Trigger Route (Float, 8):Float"(In 1=@n5.Bandwidth, In 0=@n6.Bandwidth, Crossfade Value=@n7.Value) {also {defaults {In 2=0, In 3=0, In 4=0, In 5=0, In 6=0, In 7=0}}}
   CrossfadeTriggerRouteFloat8_7 = "Crossfade.Trigger Route (Float, 8):Float"(In 1=@n5.Gain, In 0=@n6.Gain, Crossfade Value=@n7.Value) {also {defaults {In 2=0, In 3=0, In 4=0, In 5=0, In 6=0, In 7=0}}}
   CrossfadeTriggerRouteFloat8_3 = "Crossfade.Trigger Route (Float, 8):Float"(In 1=@n5.BiquadFilterTypeFloat, In 0=@n6.BiquadFilterTypeFloat, Crossfade Value=@n7.Value) {also {defaults {In 2=0, In 3=0, In 4=0, In 5=0, In 6=0, In 7=0}}}
   StereoBiquadFilter_1 = &StereoBiquadFilter(
    In L=^ExhaustL
    In R=^ExhaustR
    Cutoff Frequency=@CrossfadeTriggerRouteFloat8_1.Out
    Bandwidth=@CrossfadeTriggerRouteFloat8_4.Out
    Gain=@CrossfadeTriggerRouteFloat8_7.Out
    Type=&FloatToBiquadFilterType(Input=@CrossfadeTriggerRouteFloat8_3.Out)."Enum:BiquadFilterType"
   )
   StereoBiquadFilter_2 = &StereoBiquadFilter(In L=^InteriorL, In R=^InteriorR)
  ]
  links [
   {dst="@EngineL", src=StereoBiquadFilter_0.Out L}
   {dst="@EngineR", src=StereoBiquadFilter_0.Out R}
   {dst="@ExhaustL", src=StereoBiquadFilter_1.Out L}
   {dst="@ExhaustR", src=StereoBiquadFilter_1.Out R}
   {dst="@InteriorL", src=StereoBiquadFilter_2.Out L}
   {dst="@InteriorR", src=StereoBiquadFilter_2.Out R}
  ]
 }]
properties {assetImportData=&self:AssetImportData, bLooping=true, duration=10000, sampleRate=48000}
