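# Stereo MetaSound Source that serves as a four-band EQ effect stage for vehicle engine audio.
# Left and right audio inputs pass in series through four stereo biquad filters to the stereo output.
# Each band exposes type, frequency, gain and Q as constructor inputs, so they are set at creation, with Q converted to filter bandwidth.
# Default gains are zero, and the OnPlay trigger is unused.
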MetaSoundSource /VehicleAudio/Patches/DM_EngineAudio_PrimaryFX
inputs [
 AudioIn_L {type=Audio}
 AudioIn_R {type=Audio}
 EQ1_FilterType {type=Enum:BiquadFilterType, access=Constructor}
 EQ1_Freq {type=Float, access=Constructor}
 EQ1_Gain {type=Float, access=Constructor}
 EQ1_Q {type=Float, access=Constructor}
 EQ2_FilterType {type=Enum:BiquadFilterType, access=Constructor}
 EQ2_Freq {type=Float, access=Constructor}
 EQ2_Gain {type=Float, access=Constructor}
 EQ2_Q {type=Float, access=Constructor}
 EQ3_FilterType {type=Enum:BiquadFilterType, access=Constructor}
 EQ3_Freq {type=Float, access=Constructor}
 EQ3_Gain {type=Float, access=Constructor}
 EQ3_Q {type=Float, access=Constructor}
 EQ4_FilterType {type=Enum:BiquadFilterType, access=Constructor}
 EQ4_Freq {type=Float, access=Constructor}
 EQ4_Gain {type=Float, access=Constructor}
 EQ4_Q {type=Float, access=Constructor}
 UE.Source.OnPlay {type=Trigger}
]
interfaces [UE.OutputFormat.Stereo:1.0, UE.Source:1.1]
outputs [
 "UE.OutputFormat.Stereo.Audio:0" {type=Audio}
 "UE.OutputFormat.Stereo.Audio:1" {type=Audio}
]
pages [Default {
  input_defaults {
   AudioIn_L=null
   AudioIn_R=null
   EQ1_FilterType=4
   EQ1_Freq=100
   EQ1_Gain=0
   EQ1_Q=1
   EQ2_FilterType=4
   EQ2_Freq=300
   EQ2_Gain=0
   EQ2_Q=1
   EQ3_FilterType=4
   EQ3_Freq=5000
   EQ3_Gain=0
   EQ3_Q=1
   EQ4_FilterType=4
   EQ4_Freq=8000
   EQ4_Gain=0
   EQ4_Q=1
   UE.Source.OnPlay=null
  }
  nodes [
   ConvertFilterQToBandwidth_1 = UE.Convert Filter Q To Bandwidth(Q=^EQ4_Q)
   ConvertFilterQToBandwidth_3 = UE.Convert Filter Q To Bandwidth(Q=^EQ3_Q)
   ConvertFilterQToBandwidth_0 = UE.Convert Filter Q To Bandwidth(Q=^EQ2_Q)
   ConvertFilterQToBandwidth_2 = UE.Convert Filter Q To Bandwidth(Q=^EQ1_Q)
   n1 = &StereoBiquadFilter(
    In L=^AudioIn_L
    In R=^AudioIn_R
    Type=^EQ1_FilterType
    Cutoff Frequency=^EQ1_Freq
    Gain=^EQ1_Gain
    Bandwidth=@ConvertFilterQToBandwidth_2.Bandwidth
   )
   n2 = &StereoBiquadFilter(
    Type=^EQ2_FilterType
    Cutoff Frequency=^EQ2_Freq
    Gain=^EQ2_Gain
    Bandwidth=@ConvertFilterQToBandwidth_0.Bandwidth
    In L=@n1.Out L
    In R=@n1.Out R
   )
   n3 = &StereoBiquadFilter(
    Type=^EQ3_FilterType
    Cutoff Frequency=^EQ3_Freq
    Gain=^EQ3_Gain
    Bandwidth=@ConvertFilterQToBandwidth_3.Bandwidth
    In L=@n2.Out L
    In R=@n2.Out R
   )
   StereoBiquadFilter_2 = &StereoBiquadFilter(
    Type=^EQ4_FilterType
    Cutoff Frequency=^EQ4_Freq
    Gain=^EQ4_Gain
    Bandwidth=@ConvertFilterQToBandwidth_1.Bandwidth
    In L=@n3.Out L
    In R=@n3.Out R
   )
  ]
  links [
   {dst="@UE.OutputFormat.Stereo.Audio:0", src=StereoBiquadFilter_2.Out L}
   {dst="@UE.OutputFormat.Stereo.Audio:1", src=StereoBiquadFilter_2.Out R}
  ]
 }]
properties {
 assetImportData=&self:AssetImportData
 bLooping=true
 duration=10000
 numChannels=2
 outputFormat=Stereo
 sampleRate=48000
}
