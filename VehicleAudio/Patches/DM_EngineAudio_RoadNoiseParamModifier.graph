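# MetaSound patch in the vehicle engine audio set that modulates a float road noise parameter with random variation.
# An LFO Frequency Noise node, with slight rate jitter and a step limit, generates the modulation signal, taking its minimum value and rate from internal float variables.
# The patch multiplies the input by that signal and outputs the product.
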MetaSoundPatch /VehicleAudio/Patches/DM_EngineAudio_RoadNoiseParamModifier
inputs [Input {type=Float}]
outputs [Out {type=Float}]
pages [Default {
  input_defaults {Input=1}
  nodes [
   LfoFrequencyNoise_0 = UE.Lfo Frequency Noise:Audio(
    Min Value=VariableAccessor.Float(Variable=InitVariable.Float().Variable).Value
    Rate=VariableAccessor.Float(Variable=InitVariable.Float().Variable).Value
   ) {also {defaults {Rate Jitter=0.1, Step Limit=0.5}}}
   Multiply_0 = UE.Multiply:Float(PrimaryOperand=^Input, AdditionalOperands=@LfoFrequencyNoise_0.Out)
  ]
  links [{dst="@Out", src=Multiply_0.Out}]
 }]
