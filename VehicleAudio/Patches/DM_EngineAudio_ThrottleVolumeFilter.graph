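# MetaSound patch for vehicle engine audio that shapes a mono signal's tone and volume by throttle amount.
# A FloatEQ_Mono sub-patch uses throttle to move filter frequency and gain between off- and on-throttle settings, and exposes the cutoff frequency.
# The filtered audio is then multiplied by a gain mapped from smoothed throttle between two decibel volumes converted to linear.
# One interp time drives both filter and volume smoothing.
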MetaSoundPatch /VehicleAudio/Patches/DM_EngineAudio_ThrottleVolumeFilter
inputs [
 AudioIn {type=Audio}
 Filter.FilterQ {type=Float}
 Filter.FilterType {type=Enum:BiquadFilterType}
 Filter.OffThrottleFreq {type=Float}
 Filter.OffThrottleGain {type=Float}
 Filter.OnThrottleFreq {type=Float}
 Filter.OnThrottleGain {type=Float}
 Interp Time {type=Time}
 REV.Throttle {type=Float}
 Volume.OffThrottle {type=Float}
 Volume.OnThrottle {type=Float}
]
outputs [
 AudioOut {type=Audio}
 CutoffFrequency {type=Float}
]
pages [Default {
  input_defaults {
   AudioIn=null
   Filter.FilterQ=1
   Filter.FilterType=0
   Filter.OffThrottleFreq=20000
   Filter.OffThrottleGain=0
   Filter.OnThrottleFreq=20000
   Filter.OnThrottleGain=0
   Interp Time=0.1
   REV.Throttle=0
   Volume.OffThrottle=0
   Volume.OnThrottle=0
  }
  nodes [
   FloatEQ_Mono_0 = /FortMetasound/Patches/FloatEQ_Mono(
    AudioIn=^AudioIn
    FilterQ=^Filter.FilterQ
    FilterType=^Filter.FilterType
    MinFreq=^Filter.OffThrottleFreq
    MinGain=^Filter.OffThrottleGain
    MaxFreq=^Filter.OnThrottleFreq
    MaxGain=^Filter.OnThrottleGain
    InterpTime=^Interp Time
    FloatInput=^REV.Throttle
   )
   DecibelstoLinearGain_0 = UE.Decibels to Linear Gain:Float(Decibels=^Volume.OffThrottle)
   DecibelstoLinearGain_1 = UE.Decibels to Linear Gain:Float(Decibels=^Volume.OnThrottle)
   Multiply_0 = UE.Multiply:Audio by Float(
    PrimaryOperand=@FloatEQ_Mono_0.AudioOut
    AdditionalOperands=MapRange.MapRange:Float(
     Out Range A=@DecibelstoLinearGain_0.Linear Gain
     Out Range B=@DecibelstoLinearGain_1.Linear Gain
     In=UE.InterpTo:Audio(Interp Time=^Interp Time, Target=^REV.Throttle).Value
    ).Out Value
   )
  ]
  links [{dst="@CutoffFrequency", src=FloatEQ_Mono_0.CutoffFrequency}, {dst="@AudioOut", src=Multiply_0.Out}]
 }]
