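# Mono vehicle transmission MetaSound source that plays a wave asset loop, pitched by normalized speed through a wavetable curve and min/max range.
# Gain combines a speed volume curve and fade-in, gear number, throttle-blended levels, grounded state and master volume.
# An inverted, clamped ADSR dips the level on gear changes.
# Two chained biquad filters, one speed-swept, are built but bypassed, as the output takes the unfiltered player signal.
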MetaSoundSource /VehicleAudio/Patches/DM_EngineAudio_Transmision_Waveasset
inputs [
 GearChange_Time {type=Time, access=Constructor}
 GearChanged {type=Trigger}
 grounded {type=Bool}
 REV.EnableShifting {type=Bool}
 REV.Gear {type=Int32}
 REV.RPM {type=Float}
 REV.Throttle {type=Float}
 REV.Velocity {type=Float}
 Speed_Normalized {type=Float}
 Transmission.MaxFreq {type=Float}
 Transmission.MinFreq {type=Float}
 Transmission_OffThrottle {type=Float}
 Transmission_OnThrottle {type=Float}
 TransmissionVolume {type=Float}
 UE.Source.OnPlay {type=Trigger}
 WaveTableBank {type=WaveTableBankAsset}
]
interfaces [REV:1.0, UE.OutputFormat.Mono:1.0, UE.Source:1.1]
outputs ["UE.OutputFormat.Mono.Audio:0" {type=Audio}]
pages [Default {
  input_defaults {
   GearChange_Time=0.1
   GearChanged=null
   REV.EnableShifting=null
   REV.Gear=0
   REV.RPM=0
   REV.Throttle=1
   REV.Velocity=0
   Speed_Normalized=0.484
   Transmission.MaxFreq=2.112
   Transmission.MinFreq=-5.088
   TransmissionVolume=12
   Transmission_OffThrottle=0
   Transmission_OnThrottle=0
   UE.Source.OnPlay=null
   WaveTableBank=/VehicleAudio/Patches/Transmission_Wavetable/Transmission_Sine
   grounded="<MetaSoundBuilderResult.SUCCEEDED: 0>"
  }
  nodes [
   WavePlayer_0 = UE.Wave Player:Mono(
    Play=^UE.Source.OnPlay
    Pitch Shift=MapRange.MapRange:Float(
     Out Range B=^Transmission.MaxFreq
     Out Range A=^Transmission.MinFreq
     In=/FortMetasound/Patches/MSP_WavetableCurve(Input=^Speed_Normalized) {also {defaults {WaveTableBank=/VehicleAudio/Curves/TransmissionPitchCurve}}}.Output
    ).Out Value
   ) {also {defaults {Loop="<MetaSoundBuilderResult.SUCCEEDED: 0>", Wave Asset=&/DM_SFX_Transmission_loop_V1}, version [1,1]}}
   BiquadFilter_1 = UE.Biquad Filter:Audio(In=@WavePlayer_0.Out Mono) {also {defaults {Bandwidth=2, Cutoff Frequency=500, Gain=-9, Type=4}}}
   BiquadFilter_0 = UE.Biquad Filter:Audio(
    In=@BiquadFilter_1.Out
    Cutoff Frequency=MapRange.MapRange:Float(In=^Speed_Normalized) {also {defaults {Out Range A=1000, Out Range B=5000}}}.Out Value
   ) {also {defaults {Bandwidth=2, Gain=6, Type=0}}}
   DecibelstoLinearGain_2 = UE.Decibels to Linear Gain:Float(Decibels=^TransmissionVolume)
   DecibelstoLinearGain_0 = UE.Decibels to Linear Gain:Float(Decibels=&MSP_WavetableCurve(Input=^Speed_Normalized) {also {defaults {WaveTableBank=/VehicleAudio/Patches/Transmission_Wavetable/Trans_VolumeCurve}}}.Output)
   DecibelstoLinearGain_1 = UE.Decibels to Linear Gain:Float(Decibels=^Transmission_OffThrottle)
   DecibelstoLinearGain_3 = UE.Decibels to Linear Gain:Float(Decibels=^Transmission_OnThrottle)
   n1 = Convert.Float:Bool(Bool=^grounded)
   n2 = MapRange.MapRange:Float(In=^Speed_Normalized) {also {defaults {In Range B=0.4}}}
   TriggerDelay_0 = UE.Trigger Delay(Delay Time=^GearChange_Time, In=^GearChanged)
   ADSREnvelopeADSREnvelope_0 = ADSR Envelope.ADSR Envelope:Float(Trigger Attack=^GearChanged, Trigger Release=@TriggerDelay_0.Out) {also {defaults {Decay Time=0, Release Time=0.3, Sustain Level=0.8}}}
   n3 = Convert.Float:Int32(Int32=^REV.Gear)
   n4 = UE.Multiply:Float(
    PrimaryOperand=Clamp.Clamp:Float(In=/FortMetasound/Patches/InvertFloat(FloatIn=@ADSREnvelopeADSREnvelope_0.Out Envelope).FloatOut) {also {defaults {Max=1, Min=0.6}}}.Value
    AdditionalOperands=MapRange.MapRange:Float(In=UE.InterpTo:Audio(Target=@n3.Float).Value) {also {defaults {In Range B=7, Out Range A=0.8}}}.Out Value
   )
   n5 = UE.Multiply:Float(
    AdditionalOperands=MapRange.MapRange:Float(In=^REV.Throttle, Out Range A=@DecibelstoLinearGain_1.Linear Gain, Out Range B=@DecibelstoLinearGain_3.Linear Gain).Out Value
    PrimaryOperand=UE.Multiply:Float(
     AdditionalOperands=MapRange.MapRange:Float(In=@n1.Float) {also {defaults {Out Range A=0.2}}}.Out Value
     PrimaryOperand=UE.Multiply:Float(AdditionalOperands=@n2.Out Value, PrimaryOperand=@n4.Out).Out
    ).Out
   )
   Multiply_0 = UE.Multiply:Audio by Float(
    AdditionalOperands=UE.Multiply:Float(
     AdditionalOperands=@DecibelstoLinearGain_2.Linear Gain
     PrimaryOperand=UE.Multiply:Float(AdditionalOperands=@DecibelstoLinearGain_0.Linear Gain, PrimaryOperand=@n5.Out).Out
    ).Out
    PrimaryOperand=@WavePlayer_0.Out Mono
   )
  ]
  links [{dst="@UE.OutputFormat.Mono.Audio:0", src=Multiply_0.Out}]
 }]
properties {assetImportData=&self:AssetImportData, bLooping=true, duration=10000, sampleRate=48000}
