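# MetaSound Source that mixes vehicle engine audio, applying gain to stereo Engine, Exhaust and Interior stems and outputting them separately.
# On play it reads per-stem dB levels from exterior and interior arrays, crossfaded by an interpolated MixP and combined with each stem's volume.
# A shared REV gain combines Volume_REV with an Acceleration-driven wavetable curve.
# BonusSpeed, the mono output and the extra REV multipliers' inputs are unconnected.
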MetaSoundSource /VehicleAudio/Patches/DM_EngineAudio_VolumeMixer
inputs [
 Acceleration {type=Float}
 BonusSpeed {type=Float}
 EngineL {type=Audio}
 EngineR {type=Audio}
 ExhaustL {type=Audio}
 ExhaustR {type=Audio}
 InteriorL {type=Audio}
 InteriorR {type=Audio}
 MixP {type=Int32}
 MixP_Exterior {type=Float:Array}
 MixP_Interior {type=Float:Array}
 REV_AccelerationCurves {type=WaveTableBankAsset}
 UE.Source.OnPlay {type=Trigger}
 Volume_Engine {type=Float}
 Volume_Exhaust {type=Float}
 Volume_Interior {type=Float}
 Volume_REV {type=Float}
]
interfaces [UE.OutputFormat.Mono:1.0, UE.Source:1.1]
outputs [
 EngineL {type=Audio}
 EngineR {type=Audio}
 ExhaustL {type=Audio}
 ExhaustR {type=Audio}
 InteriorL {type=Audio}
 InteriorR {type=Audio}
 "UE.OutputFormat.Mono.Audio:0" {type=Audio}
]
pages [Default {
  input_defaults {
   Acceleration=0
   BonusSpeed=0
   EngineL=null
   EngineR=null
   ExhaustL=null
   ExhaustR=null
   InteriorL=null
   InteriorR=null
   MixP=0
   REV_AccelerationCurves=/VehicleAudio/EngineAudio/Common/REV_AccelerationCurve
   UE.Source.OnPlay=null
   Volume_Engine=1
   Volume_Exhaust=1
   Volume_Interior=1
   Volume_REV=1
   MixP_Exterior [0,0,0]
   MixP_Interior [-10,-10,-10]
  }
  nodes [
   DecibelstoLinearGain_7 = UE.Decibels to Linear Gain:Float(Decibels=^Volume_Exhaust)
   n1 = TriggerCompare.Clamp:Int32(Compare=^UE.Source.OnPlay, A=Array.Num:Float:Array(Array=^MixP_Interior).Num) {also {defaults {Type=1}}}
   n2 = TriggerCompare.Clamp:Int32(Compare=^UE.Source.OnPlay, A=Array.Num:Float:Array(Array=^MixP_Exterior).Num) {also {defaults {Type=1}}}
   n3 = UE.InterpTo:Audio(Target=Convert.Float:Int32(Int32=^MixP).Float)
   CrossfadeTriggerRouteFloat8_1 = "Crossfade.Trigger Route (Float, 8):Float"(
    In 1=Array.Get:Float:Array(Array=^MixP_Interior, Trigger=@n1.True) {also {defaults {Index=0}}}.Element
    In 0=Array.Get:Float:Array(Array=^MixP_Exterior, Trigger=@n2.True).Element
    Crossfade Value=@n3.Value
   )
   DecibelstoLinearGain_1 = UE.Decibels to Linear Gain:Float(Decibels=@CrossfadeTriggerRouteFloat8_1.Out)
   n4 = VariableAccessor.Float(Variable=VariableMutator.Float(Value=@DecibelstoLinearGain_1.Linear Gain, Variable=InitVariable.Float().Variable).Variable)
   n5 = VariableAccessor.Float(Variable=VariableMutator.Float(
     Variable=InitVariable.Float().Variable
     Value=UE.Multiply:Float(PrimaryOperand=@DecibelstoLinearGain_7.Linear Gain, AdditionalOperands=@n4.Value).Out
    ).Variable)
   DecibelstoLinearGain_0 = UE.Decibels to Linear Gain:Float(Decibels=MapRange.MapRange:Float(In=/FortMetasound/Patches/MSP_WavetableCurve(Input=^Acceleration, WaveTableBank=^REV_AccelerationCurves).Output) {also {defaults {In Range A=-1, In Range B=0, Out Range A=-2, Out Range B=0}}}.Out Value)
   n6 = Convert.Float:Bool()
   DecibelstoLinearGain_2 = UE.Decibels to Linear Gain:Float(Decibels=^Volume_REV)
   n7 = MapRange.MapRange:Float(In=Convert.Float:Bool().Float) {also {defaults {Out Range A=1, Out Range B=1.3}}}
   n8 = UE.Multiply:Float(
    AdditionalOperands=MapRange.MapRange:Float() {also {defaults {In=0, Out Range A=1, Out Range B=1.3}}}.Out Value
    PrimaryOperand=UE.Multiply:Float(
     AdditionalOperands=MapRange.MapRange:Float(In=@n6.Float) {also {defaults {Out Range A=1, Out Range B=0.85}}}.Out Value
     PrimaryOperand=UE.Multiply:Float(PrimaryOperand=@DecibelstoLinearGain_2.Linear Gain, AdditionalOperands=@n7.Out Value).Out
    ).Out
   )
   n9 = VariableAccessor.Float(Variable=VariableMutator.Float(
     Variable=InitVariable.Float().Variable
     Value=UE.Multiply:Float(AdditionalOperands=@DecibelstoLinearGain_0.Linear Gain, PrimaryOperand=@n8.Out).Out
    ).Variable)
   Multiply_0 = UE.Multiply:Audio by Float(PrimaryOperand=UE.Multiply:Audio by Float(PrimaryOperand=^ExhaustL, AdditionalOperands=@n5.Value).Out, AdditionalOperands=@n9.Value)
   DecibelstoLinearGain_6 = UE.Decibels to Linear Gain:Float(Decibels=^Volume_Interior)
   CrossfadeTriggerRouteFloat8_2 = "Crossfade.Trigger Route (Float, 8):Float"(
    In 0=Array.Get:Float:Array(Array=^MixP_Exterior, Trigger=@n2.True) {also {defaults {Index=2}}}.Element
    In 1=Array.Get:Float:Array(Array=^MixP_Interior, Trigger=@n1.True) {also {defaults {Index=2}}}.Element
    Crossfade Value=@n3.Value
   ) {also {defaults {In 2=0, In 3=0, In 4=0, In 5=0, In 6=0, In 7=0}}}
   DecibelstoLinearGain_5 = UE.Decibels to Linear Gain:Float(Decibels=@CrossfadeTriggerRouteFloat8_2.Out)
   n10 = VariableAccessor.Float(Variable=VariableMutator.Float(Value=@DecibelstoLinearGain_5.Linear Gain, Variable=InitVariable.Float().Variable).Variable)
   n11 = VariableAccessor.Float(Variable=VariableMutator.Float(
     Variable=InitVariable.Float().Variable
     Value=UE.Multiply:Float(PrimaryOperand=@DecibelstoLinearGain_6.Linear Gain, AdditionalOperands=@n10.Value).Out
    ).Variable)
   n12 = VariableAccessor.Float(Variable=@n9.Variable)
   n13 = VariableAccessor.Float(Variable=@n12.Variable)
   n14 = VariableAccessor.Float(Variable=@n13.Variable)
   n15 = VariableAccessor.Float(Variable=@n14.Variable)
   Multiply_10 = UE.Multiply:Audio by Float(PrimaryOperand=UE.Multiply:Audio by Float(PrimaryOperand=^InteriorR, AdditionalOperands=@n11.Value).Out, AdditionalOperands=@n15.Value)
   Multiply_11 = UE.Multiply:Audio by Float(PrimaryOperand=UE.Multiply:Audio by Float(PrimaryOperand=^InteriorL, AdditionalOperands=@n11.Value).Out, AdditionalOperands=@n14.Value)
   DecibelstoLinearGain_3 = UE.Decibels to Linear Gain:Float(Decibels=^Volume_Engine)
   CrossfadeTriggerRouteFloat8_0 = "Crossfade.Trigger Route (Float, 8):Float"(
    In 1=Array.Get:Float:Array(Array=^MixP_Interior, Trigger=@n1.True) {also {defaults {Index=1}}}.Element
    In 0=Array.Get:Float:Array(Array=^MixP_Exterior, Trigger=@n2.True) {also {defaults {Index=1}}}.Element
    Crossfade Value=@n3.Value
   ) {also {defaults {In 2=0, In 3=0, In 4=0, In 5=0, In 6=0, In 7=0}}}
   DecibelstoLinearGain_4 = UE.Decibels to Linear Gain:Float(Decibels=@CrossfadeTriggerRouteFloat8_0.Out)
   n16 = VariableAccessor.Float(Variable=VariableMutator.Float(Value=@DecibelstoLinearGain_4.Linear Gain, Variable=InitVariable.Float().Variable).Variable)
   n17 = VariableAccessor.Float(Variable=VariableMutator.Float(
     Variable=InitVariable.Float().Variable
     Value=UE.Multiply:Float(PrimaryOperand=@DecibelstoLinearGain_3.Linear Gain, AdditionalOperands=@n16.Value).Out
    ).Variable)
   Multiply_3 = UE.Multiply:Audio by Float(PrimaryOperand=UE.Multiply:Audio by Float(PrimaryOperand=^EngineR, AdditionalOperands=@n17.Value).Out, AdditionalOperands=@n13.Value)
   Multiply_5 = UE.Multiply:Audio by Float(PrimaryOperand=UE.Multiply:Audio by Float(PrimaryOperand=^EngineL, AdditionalOperands=@n17.Value).Out, AdditionalOperands=@n12.Value)
   Multiply_9 = UE.Multiply:Audio by Float(
    PrimaryOperand=UE.Multiply:Audio by Float(PrimaryOperand=^ExhaustR, AdditionalOperands=VariableAccessor.Float(Variable=@n5.Variable).Value).Out
    AdditionalOperands=VariableAccessor.Float(Variable=@n15.Variable).Value
   )
  ]
  links [
   {dst="@ExhaustL", src=Multiply_0.Out}
   {dst="@InteriorR", src=Multiply_10.Out}
   {dst="@InteriorL", src=Multiply_11.Out}
   {dst="@EngineR", src=Multiply_3.Out}
   {dst="@EngineL", src=Multiply_5.Out}
   {dst="@ExhaustR", src=Multiply_9.Out}
  ]
 }]
properties {assetImportData=&self:AssetImportData, bLooping=true, duration=10000, sampleRate=48000}
