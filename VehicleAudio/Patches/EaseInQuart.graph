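# MetaSound patch in the vehicle audio folder that shapes a float input with a steep ease-in power curve.
# Three chained Multiply nodes each square the previous value, and the last feeds the output.
# The result is the input to the eighth power, steeper than the fourth-power curve the EaseInQuart name implies.
# No clamping is applied, so inputs beyond the 0 to 1 range in magnitude grow rapidly.
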MetaSoundPatch /VehicleAudio/Patches/EaseInQuart
inputs [Input {type=Float}]
outputs [Output {type=Float}]
pages [Default {
  input_defaults {Input=0}
  nodes [
   n1 = UE.Multiply:Float(AdditionalOperands=^Input, PrimaryOperand=^Input)
   n2 = UE.Multiply:Float(AdditionalOperands=@n1.Out, PrimaryOperand=@n1.Out)
   Multiply_1 = UE.Multiply:Float(AdditionalOperands=@n2.Out, PrimaryOperand=@n2.Out)
  ]
  links [{dst="@Output", src=Multiply_1.Out}]
 }]
