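# EaseOutQuad is a MetaSound patch in VehicleAudio that applies a quadratic ease-out curve to a float input.
# It computes 1 - (1 - In) * (1 - In) using Subtract and Multiply nodes, with two separate Subtract nodes producing the inverted input for the multiply.
# For inputs from 0 to 1, the output rises quickly and then flattens toward 1.
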MetaSoundPatch /VehicleAudio/Patches/EaseOutQuad
inputs [In {type=Float}]
interfaces [MetaSound:1.0]
outputs [Out {type=Float}]
pages [Default {
  input_defaults {In=0}
  nodes [Subtract_0 = UE.Subtract:Float(AdditionalOperands=UE.Multiply:Float(
     PrimaryOperand=UE.Subtract:Float(AdditionalOperands=^In) {also {defaults {PrimaryOperand=1}}}.Out
     AdditionalOperands=UE.Subtract:Float(AdditionalOperands=^In) {also {defaults {PrimaryOperand=1}}}.Out
    ).Out) {also {defaults {PrimaryOperand=1}}}]
  links [{dst="@Out", src=Subtract_0.Out}]
 }]
