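# MetaSound patch that adds throttle-dependent waveshaper distortion to vehicle engine audio, implementing the REV interface but reading only its Throttle input.
# A band splitter leaves the low band clean and sends the high band through a waveshaper, crossfaded against its dry signal by Mix.
# Throttle maps between off- and on-throttle scales that multiply distortion Amount.
# Bands are remixed, and Bypass crossfades to the untouched input.
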MetaSoundPatch /VehicleAudio/Patches/MSP_EngineAudio_Distortion
inputs [
 Input {type=Audio}
 OffThrottleDistortionScale {type=Float}
 OnThrottleDistortionScale {type=Float}
 REV.EnableShifting {type=Bool}
 REV.Gear {type=Int32}
 REV.RPM {type=Float}
 REV.Throttle {type=Float}
 REV.Velocity {type=Float}
 Waveshaper.Amount {type=Float}
 Waveshaper.Bias {type=Float}
 Waveshaper.Bypass {type=Bool}
 Waveshaper.Crossover {type=Float}
 Waveshaper.Mix {type=Float}
 Waveshaper.OutputGain {type=Float}
 Waveshaper.Type {type=Enum:WaveShaperType}
]
interfaces [REV:1.0]
outputs [Input {type=Audio}]
pages [Default {
  input_defaults {
   Input=null
   OffThrottleDistortionScale=0
   OnThrottleDistortionScale=1
   REV.EnableShifting=null
   REV.Gear=1
   REV.RPM=0
   REV.Throttle=0
   REV.Velocity=0
   Waveshaper.Amount=10
   Waveshaper.Bias=0
   Waveshaper.Bypass=null
   Waveshaper.Crossover=272
   Waveshaper.Mix=0
   Waveshaper.OutputGain=1
   Waveshaper.Type=0
  }
  nodes [
   BandSplitterBandSplitterMono2_0 = "BandSplitter.Band Splitter (Mono, 2)"(In=^Input, Crossover 0=^Waveshaper.Crossover)
   n1 = MapRange.MapRange:Float(Out Range A=^OffThrottleDistortionScale, Out Range B=^OnThrottleDistortionScale, In=^REV.Throttle)
   CrossfadeTriggerRouteAudio2_1 = "Crossfade.Trigger Route (Audio, 2):Audio"(
    Crossfade Value=^Waveshaper.Mix
    In 0=@BandSplitterBandSplitterMono2_0.Band 1 Out
    In 1=UE.WaveShaper:Audio(
     Bias=^Waveshaper.Bias
     OutputGain=^Waveshaper.OutputGain
     Type=^Waveshaper.Type
     In=@BandSplitterBandSplitterMono2_0.Band 1 Out
     Amount=UE.Multiply:Float(PrimaryOperand=^Waveshaper.Amount, AdditionalOperands=@n1.Out Value).Out
    ).Out
   )
   AudioMixerAudioMixerMono2_0 = "AudioMixer.Audio Mixer (Mono, 2)"(In 0=@BandSplitterBandSplitterMono2_0.Band 0 Out, In 1=@CrossfadeTriggerRouteAudio2_1.Out)
   CrossfadeTriggerRouteAudio2_0 = "Crossfade.Trigger Route (Audio, 2):Audio"(In 1=^Input, In 0=@AudioMixerAudioMixerMono2_0.Out, Crossfade Value=Convert.Float:Bool(Bool=^Waveshaper.Bypass).Float)
  ]
  links [{dst="@Input", src=CrossfadeTriggerRouteAudio2_0.Out}]
 }]
