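# One-shot mono MetaSound source for a vehicle engine-off sound, spatialized with vehicle movement attenuation.
# On play, a check on the input wave array's length either triggers a shuffled wave pick or fires OnFinished immediately.
# The wave player's output is scaled by a decibel volume input, and its completion also ends the source.
# Vehicle engine modulators drive volume and filters, with engine-off and ignition-off concurrency applied.
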MetaSoundSource /VehicleAudio/Patches/MSS_EngineAudio_Off
inputs [
 In Array {type=WaveAsset:Array, access=Constructor}
 OutputVolume {type=Float}
 UE.Source.OnPlay {type=Trigger}
]
interfaces [UE.OutputFormat.Mono:1.0, UE.Source.OneShot:1.0, UE.Source:1.1]
outputs [
 "UE.OutputFormat.Mono.Audio:0" {type=Audio}
 UE.Source.OneShot.OnFinished {type=Trigger}
]
pages [Default {
  input_defaults {OutputVolume=0, UE.Source.OnPlay=null, In Array []}
  nodes [
   DecibelstoLinearGain_0 = UE.Decibels to Linear Gain:Float(Decibels=^OutputVolume)
   n1 = TriggerCompare.Clamp:Int32(Compare=^UE.Source.OnPlay, A=Array.Num:WaveAsset:Array(Array=^In Array).Num) {also {defaults {Type=3}}}
   n2 = Array.Shuffle:WaveAsset:Array(In Array=^In Array, Next=@n1.TRUE)
   WavePlayer_0 = UE.Wave Player:Mono(Play=@n2.On Next, Wave Asset=@n2.Value) {also {version [1,1]}}
   Multiply_0 = UE.Multiply:Audio by Float(AdditionalOperands=@DecibelstoLinearGain_0.Linear Gain, PrimaryOperand=@WavePlayer_0.Out Mono)
   TriggerAnyTriggerAny2_0 = "TriggerAny.Trigger Any (2)"(In 1=@n1.FALSE, In 0=@WavePlayer_0.On Finished)
  ]
  links [{dst="@UE.OutputFormat.Mono.Audio:0", src=Multiply_0.Out}, {dst="@UE.Source.OneShot.OnFinished", src=TriggerAnyTriggerAny2_0.Out}]
 }]
properties {
 assetImportData=&self:AssetImportData
 attenuationSettings=/VehicleAudio/Settings/Vehicle_Movement_Att
 audioPropertiesSheet=/VehicleAudio/PropertySheets/APS_Cosmetic_Vehicle_Engine_Off
 maxDistance=12512
 sampleRate=48000
 soundClassObject=&Fort_Vehicles_Exterior
 soundSubmixObject=&Fortnite_SFX
 concurrencySet [/VehicleAudio/Settings/Vehicle_Cosmetic_Off_Conc, /VehicleAudio/Settings/Vehicle_Cosmetic_Ignition_Off_Conc]
 modulationSettings {
  highpassRouting=Union
  lowpassRouting=Union
  volumeRouting=Union
  highpassModulationDestination {bEnableModulation=true, modulators [&PP_FortVehicles_Engine_HPF]}
  lowpassModulationDestination {bEnableModulation=true, modulators [&Patch_FortVehicles_Engine_LPF]}
  volumeModulationDestination {
   bEnableModulation=true
   modulators [&Patch_FortVehicles_Engine_Vol, &DM_VehicleEnginePatch]
  }
 }
}
