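# Mono MetaSound source that plays a vehicle horn from a supplied wave asset, with horn-specific attenuation and an exterior vehicle sound class.
# On play, a Wave Player starts the wave with looping set by an input, and its output is scaled by a constructor decibel volume converted to linear gain.
# The OnStop, DecayTime and DecayCurve inputs are declared but unconnected, so the graph applies no release envelope.
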MetaSoundSource /VehicleAudio/Patches/MSS_Vehicle_Horn
inputs [
 DecayCurve {type=Float}
 DecayTime {type=Time}
 Looping {type=Bool}
 OnStop {type=Trigger}
 OutputVolume {type=Float, access=Constructor}
 UE.Source.OnPlay {type=Trigger}
 Wave Asset {type=WaveAsset}
]
interfaces [UE.OutputFormat.Mono:1.0, UE.Source:1.1]
outputs ["UE.OutputFormat.Mono.Audio:0" {type=Audio}]
pages [Default {
  input_defaults {DecayCurve=1.5, DecayTime=0.1, Looping=null, OnStop=null, OutputVolume=0, UE.Source.OnPlay=null, Wave Asset=null}
  nodes [
   DecibelstoLinearGain_0 = UE.Decibels to Linear Gain:Float(Decibels=^OutputVolume)
   WavePlayer_0 = UE.Wave Player:Mono(Loop=^Looping, Play=^UE.Source.OnPlay, Wave Asset=^Wave Asset) {also {version [1,1]}}
   Multiply_0 = UE.Multiply:Audio by Float(AdditionalOperands=@DecibelstoLinearGain_0.Linear Gain, PrimaryOperand=@WavePlayer_0.Out Mono)
  ]
  links [{dst="@UE.OutputFormat.Mono.Audio:0", src=Multiply_0.Out}]
 }]
properties {
 assetImportData=&self:AssetImportData
 attenuationSettings=/VehicleAudio/Settings/Vehicle_Horn_Att
 audioPropertiesSheet=/VehicleAudio/PropertySheets/APS_Cosmetic_Vehicle_Horn
 bLooping=true
 duration=10000
 maxDistance=16030.001
 sampleRate=48000
 soundClassObject=&Fort_Vehicles_Exterior
 volume=1.585
}
